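; Placeable weapon upgrade bench (Wumba) that upgrades or sidegrades a player's weapon, with sidegrading enabled once playlist data is ready.
; Interaction times follow held weapon rarity; success sends quest stats, applies a gameplay effect and plays rarity-tiered FX.
; On placement it pings and shows the placer's name; overlap volumes toggle client-side prompt widgets.
; Pickaxe hits wobble it, death can drop a pickup, and losing its support triggers relinking or destruction.
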(event EventBeginPlay
  (bind (_displayname _icon _customindicatorclass _customdetailsclass _primarycolor _sound _battachsound _screenclamping) (Utilities|Struct|BreakMarkedActorDisplayInfo (Variables|Marking|GetMarkerDisplay)))
  (bind _horizontalenabled (Variables|Default|GetHorizontalEnabled))
  (bind _returnvalue (ScalableFloat|GetValueatLevel _horizontalenabled))
  (bind _returnvalue_2 (== _returnvalue 1.0))
  (bind _returnvalue_3 (Game|GetInstigator))
  (bind _returnvalue_4 (Fortnite|GetFortPlayerControllerfromActor _returnvalue_3))
  (bind _vendingmachine_ambient (Variables|Default|GetVendingMachine_Ambient))
  ;[1 1 1] BEGIN PLAY
  (Audio|PlaySoundatLocation "/Game/Sounds/Foley_Loot/Containers/Wumba/Wumba_Placed_Cue.Wumba_Placed_Cue" (Transformation|GetActorLocation))
  (Audio|PlaySoundatLocation "/Game/Sounds/Foley_Loot/Containers/Wumba/WumbaMachine_Deploy.WumbaMachine_Deploy" (Transformation|GetActorLocation))
  (Utilities|FlowControl|Sequence
    (:then_0
      ;[1 1 1] Set up audio
      (Audio|Components|Audio|SetSound _vendingmachine_ambient (Variables|Sound|GetAmbient_Sound))
      (Audio|Components|Audio|Play _vendingmachine_ambient 0.0)
      (Audio|Components|Audio|Play (Variables|Default|GetFlashing_Light_Sound_Loop) 0.0)
      ;[1 1 1] UI Refresh
      (CallFunction|InitializeUI)
      (Effects|Components|ParticleSystem|SetFloatParameter (Variables|Default|GetP_TableLightIdle) "LightSize" (Variables|Default|GetLightSize))
      ;[1 1 1] Add player's name to display name
      (Utilities|IsValid _returnvalue_4
        (:Is%20Valid
          (Variables|Marking|SetMarkerDisplay (Utilities|Struct|MakeMarkedActorDisplayInfo (Utilities|Text|FormatText (Variables|Default|GetUpgradeBenchDisplayName) (Fortnite|GetPlayerNameSafe _returnvalue_4)) _icon _customindicatorclass 0 _primarycolor _sound true _screenclamping)))
        (:Is%20Not%20Valid
          _))
      ;[1 1 1] Ping when placed
      (Networking|SwitchHasAuthority
        (:Authority
          (if (not (Variables|Default|GetDoOnce))
            (Variables|Default|SetDoOnce true)
            (CallFunction|PingMultiCast)))))
    (:then_1
      (Networking|SwitchHasAuthority
        (:Authority
          (Class|FortAsyncActionCurrentPlaylistReady|PlaylistDataReadyAsync
            (:OnReady
              (Variables|Interaction|SetbAllowSecondInteraction _returnvalue_2)
              (Variables|Interaction|SetbShowSecondInteractPrompt _returnvalue_2))))
        (:Remote
          (Variables|Interaction|SetbShowSecondInteractPrompt (Variables|Interaction|GetbAllowSecondInteraction))))
      (Variables|Actor|SetbCanBeDamaged (== (ScalableFloat|GetValueatLevel (Variables|Default|GetAllowDamage)) 1.0)))))

(event Custom|PingMultiCast
  ;[1 1 1] Ping when placed
  (bind _asfort_player_controller_athena (Utilities|Casting|CastToFortPlayerControllerAthena (Fortnite|GetFortPlayerControllerfromActor (Game|GetInstigator))))
  (Fortnite|RegisterSoundIndicatorPlacementContext _asfort_player_controller_athena self))

(event Interaction|EventBlueprintOnBeginInteract
  (params (InteractingPawn :type Fort%20Pawn%20Object%20Reference) (InteractionBeingAttempted :type EInteractionBeingAttempted%20Enum :default "FirstInteraction") (InteractionType :type TInteractionType%20Enum :default "IT_NoInteraction"))
  ;[1 1 1] Search feedback
  (Audio|Components|Audio|Play (Variables|Default|GetSearch_Sound_Loop) 0.0))

(event Custom|WobbleMachine
  ;[1 1 1] Make Machine Wobble from Animation
  (if (not (Variables|Default|GetUpgradeAnim))
    (Variables|Default|SetUpgradeAnim true)
    (CallFunction|UpdateAnimInstanceVisuals)
    (Utilities|FlowControl|Delay 1.16) ; anim Length
    (Variables|Default|SetUpgradeAnim false)
    (CallFunction|UpdateAnimInstanceVisuals)))

(event EventOnDamagePlayEffects
  (params (Damage :type Float%20%28single-precision%29) (DamageTags :type Gameplay%20Tag%20Container%20Structure%20%28by%20ref%29) (Momentum :type Vector :default "0, 0, 0") (HitInfo :type Hit%20Result%20Structure%20%28by%20ref%29) (InstigatedBy :type Fort%20Pawn%20Object%20Reference) (DamageCauser :type Actor%20Object%20Reference) (EffectContext :type Gameplay%20Effect%20Context%20Handle%20Structure))
  ;[1 1 1] Wobble on Hit
  (Utilities|Casting|CastToFortWeaponPickaxeAthena DamageCauser)
  (CallFunction|WobbleMachine)
  (Audio|SpawnSoundatLocation (Variables|Sound|GetImpact_Wobble) (Transformation|GetActorLocation)))

(event Interaction|EventBlueprintOnInteractInterrupted
  ;[1 1 1] Interaction Interupted
  (CallFunction|StopSearchSoundLoop))

(event Interaction|EventBlueprintOnInteract
  (params (InteractingPawn :type Fort%20Pawn%20Object%20Reference) (InteractionBeingAttempted :type EInteractionBeingAttempted%20Enum :default "FirstInteraction"))
  (bind _upgradedweaponrarity (Variables|InteractMethods|GetUpgradedWeaponRarity))
  (if (Variables|InteractMethods|GetbCurrentInteractionSuccess)
    (Utilities|FlowControl|Sequence
      (:then_0
        ; QuestHook
        (Networking|SwitchHasAuthority
          (:Authority
            (Quest|SendComplexCustomStatEvent (Quests|GetQuestManager (Fortnite|GetFortPlayerControllerfromActor InteractingPawn) "Invalid") self (Utilities|Select :Index (Variables|InteractMethods|GetUpgradedWeaponRarity) :Common "(GameplayTags=((TagName=\"Athena.Quests.UseUpgradeBench.Common\")))" :Uncommon "(GameplayTags=((TagName=\"Athena.Quests.UseUpgradeBench.Uncommon\")))" :Rare "(GameplayTags=((TagName=\"Athena.Quests.UseUpgradeBench.Rare\")))" :Epic "(GameplayTags=((TagName=\"Athena.Quests.UseUpgradeBench.Epic\")))" :Legendary "(GameplayTags=((TagName=\"Athena.Quests.UseUpgradeBench.Legendary\")))") (Utilities|Select :Index InteractionBeingAttempted :FirstInteraction (Variables|Default|GetUpgradeQuestTag) :SecondInteraction (Variables|Default|GetSidegradeQuestTag)) 1))))
      (:then_1
        (CallFunction|PlayUpgradeSuccess (Utilities|Select :Index _upgradedweaponrarity :Common "0, 0, 0" :Uncommon "0.000000,2.000000,0.000000" :Rare "0,0.250000,2.500000" :Epic "1.250000, 0,2.500000" :Legendary "2.250000,1.125000, 0" :Mythic "0, 0, 0" :Transcendent "0, 0, 0" :Unattainable "0, 0, 0") _upgradedweaponrarity InteractionBeingAttempted)
        (GameplayEffects|ApplyGameplayEffectToTarget :self (Ability|GetAbilitySystemComponent InteractingPawn) :GameplayEffectClass "/Game/Athena/Items/Gameplay/Wumba/GE_WumbaUsed.GE_WumbaUsed_C" :Target (Ability|GetAbilitySystemComponent InteractingPawn) :Level (Math|Conversions|ToFloat(Byte) _upgradedweaponrarity))))
    (else
      (CallFunction|PlayVendFailFX))))

(event EventDestroyed
  ;[1 1 1] Interaction Interupted
  (bind _search_audio_component (Variables|Default|GetSearch%20Audio%20Component)
    (:then
      (if (Audio|Components|Audio|IsPlaying _search_audio_component)
        (Audio|Components|Audio|Stop _search_audio_component)))
    (:else)))

(event Custom|PlayVendFailFX
  (Audio|PlaySoundatLocation (Variables|VendingFX|GetVendFailedSound) (Transformation|GetActorLocation))
  (Variables|Default|SetVendSuccess false)
  (Components|Activation|Activate (Variables|Default|GetForceFeedbackFail) false)
  (CallFunction|StopSearchSoundLoop))

(event Custom|PlayUpgradeSuccess
  (params (Color :type Vector) (Index :type Byte) (InteractionBeingAttempted :type EInteractionBeingAttempted%20Enum :default "FirstInteraction"))
  (bind _p_tablelightidle (Variables|Default|GetP_TableLightIdle))
  (bind _p_upgradefx (Variables|Default|GetP_UpgradeFX))
  (Utilities|FlowControl|Sequence
    (:then_0
      (CallFunction|WobbleMachine)
      (CallFunction|StopSearchSoundLoop))
    (:then_1
      ;[1 1 1] Update Pacticle Color (included particle light)
      (Components|Activation|Activate (Variables|Default|GetP_UpgradeFX) true)
      (Effects|Components|ParticleSystem|SetVectorParameter _p_upgradefx "Particle Color" Color)
      (CallFunction|StopSearchSoundLoop)
      (Utilities|FlowControl|Sequence
        (:then_0
          (switch int (Math|Conversions|ToInteger(Byte) Index)
            (:1
              (Effects|Components|ParticleSystem|SetEmitterEnable _p_upgradefx "Rare" false)
              (Effects|Components|ParticleSystem|SetEmitterEnable _p_upgradefx "Epic" false)
              (Effects|Components|ParticleSystem|SetEmitterEnable _p_upgradefx "Legendary" false)
              (Audio|PlaySoundatLocation (Utilities|Select :Index InteractionBeingAttempted :FirstInteraction (Variables|Sound|GetActivate_Uncommon_Sound) :SecondInteraction (Variables|Sound|GetActivate_Sidegrade_Sound)) (Transformation|GetActorLocation))
              (Development|PrintString "Uncommon" true true "(R=0.000000,G=1.000000,B=0.022145,A=1.000000)" 5.0))
            (:2
              (Effects|Components|ParticleSystem|SetEmitterEnable _p_upgradefx "Rare" true)
              (Effects|Components|ParticleSystem|SetEmitterEnable _p_upgradefx "Epic" false)
              (Effects|Components|ParticleSystem|SetEmitterEnable _p_upgradefx "Legendary" false)
              (Audio|PlaySoundatLocation (Variables|Sound|GetActivate_Rare_Sound) (Transformation|GetActorLocation))
              (Development|PrintString "Rare" true true "(R=0.000000,G=0.141572,B=1.000000,A=1.000000)" 5.0))
            (:3
              (Effects|Components|ParticleSystem|SetEmitterEnable _p_upgradefx "Rare" true)
              (Effects|Components|ParticleSystem|SetEmitterEnable _p_upgradefx "Epic" true)
              (Effects|Components|ParticleSystem|SetEmitterEnable _p_upgradefx "Legendary" false)
              (Audio|PlaySoundatLocation (Variables|Sound|GetActivate_Epic_Sound) (Transformation|GetActorLocation))
              (Development|PrintString "Epic" true true "(R=0.722947,G=0.000000,B=1.000000,A=1.000000)" 5.0))
            (:4
              (Effects|Components|ParticleSystem|SetEmitterEnable _p_upgradefx "Rare" true)
              (Effects|Components|ParticleSystem|SetEmitterEnable _p_upgradefx "Epic" true)
              (Effects|Components|ParticleSystem|SetEmitterEnable _p_upgradefx "Legendary" true)
              (Audio|PlaySoundatLocation (Variables|Sound|GetActivate_Legendary_Sound) (Transformation|GetActorLocation))
              (Development|PrintString "Legendary" true true "(R=1.000000,G=0.474049,B=0.000000,A=1.000000)" 5.0))))
        (:then_1
          (Components|Activation|Deactivate _p_tablelightidle)
          (Audio|Components|Audio|FadeOut (Variables|Default|GetFlashing_Light_Sound_Loop) 0.1 0.0 "Linear" :!class /Script/Engine.AudioComponent)
          (Utilities|FlowControl|RetriggerableDelay 1.5)
          (Components|Activation|Activate _p_tablelightidle true)
          (Audio|Components|Audio|Play (Variables|Default|GetFlashing_Light_Sound_Loop) 0.0))))))

(event Custom|StopSearchSoundLoop
  (bind _search_sound_loop (Variables|Default|GetSearch_Sound_Loop))
  (if (Audio|Components|Audio|IsPlaying _search_sound_loop)
    (Audio|Components|Audio|FadeOut _search_sound_loop (Variables|Sound|GetSearch%20Sound%20Fade%20out%20Duration) 0.0 "Linear" :!class /Script/Engine.AudioComponent)))

(event EventOnDeathServer
  (params (Damage :type Float%20%28single-precision%29) (DamageTags :type Gameplay%20Tag%20Container%20Structure%20%28by%20ref%29) (Momentum :type Vector :default "0, 0, 0") (HitInfo :type Hit%20Result%20Structure%20%28by%20ref%29) (InstigatedBy :type Controller%20Object%20Reference) (DamageCauser :type Actor%20Object%20Reference) (EffectContext :type Gameplay%20Effect%20Context%20Handle%20Structure))
  ;[1 1 1] On Death, Spawn a Pick Up
  (Utilities|FlowControl|DoOnce false
    (:Completed
      (if (> (ScalableFloat|GetValueatLevel (Variables|Default|GetSpawnPickUp)) 0.0)
        (Inventory|SpawnPickupInWorld "/Game/Athena/Items/Gameplay/Wumba/AGID_Wumba.AGID_Wumba" 1 (+ (Transformation|GetActorLocation) "0, 0,200.000000") "0.000000,0.000000,1.000000" 0 true true false -1 "Tossed" "Unset" 0 false)))))

(event EventOnDeathPlayEffects
  (params (Damage :type Float%20%28single-precision%29) (DamageTags :type Gameplay%20Tag%20Container%20Structure%20%28by%20ref%29) (Momentum :type Vector :default "0, 0, 0") (HitInfo :type Hit%20Result%20Structure%20%28by%20ref%29) (InstigatedBy :type Fort%20Pawn%20Object%20Reference) (DamageCauser :type Actor%20Object%20Reference) (EffectContext :type Gameplay%20Effect%20Context%20Handle%20Structure))
  ;[1 1 1] CUSTOM EVENT -> DoVendDeath
  ;;[0.1742 1 0.1335] Do death FX.
  ;; Adjust lifespan time to match.
  (Audio|PlaySoundatLocation (Variables|Sound|GetDestroyedSound) (Transformation|GetActorLocation))
  (Components|Activation|SetActive (Variables|Default|GetP_Destroyed) true true)
  (Rendering|SetVisibility (Variables|Default|GetSK_Wumba) false false)
  (Components|Activation|Deactivate (Variables|Default|GetP_TableLightIdle))
  (Audio|Components|Audio|FadeOut (Variables|Default|GetFlashing_Light_Sound_Loop) 0.1 0.0 "Linear" :!class /Script/Engine.AudioComponent))

(event OnComponentBeginOverlap(SphereDetail)
  (params (OverlappedComponent :type Primitive%20Component%20Object%20Reference) (OtherActor :type Actor%20Object%20Reference) (OtherComp :type Primitive%20Component%20Object%20Reference) (OtherBodyIndex :type Integer) (bFromSweep :type Boolean) (SweepResult :type Hit%20Result%20Structure%20%28by%20ref%29))
  (CallFunction|CheckShouldDisplayUI true true))

(event Custom|CheckShouldDisplayUI
  (params (ShowDetail :type Boolean) (ShowDistance :type Boolean))
  (bind _ui_interactionpoint_upgrade (Variables|Default|GetUI_InteractionPoint_Upgrade))
  (bind _returnvalue (UserInterface|GetUserWidgetObject _ui_interactionpoint_upgrade))
  ;[1 1 1] Toggle "Upgrade" Widget
  (if (not (Networking|IsServer))
    (Utilities|IsValid _returnvalue
      (:Is%20Valid
        (bind _asfort_player_controller (Utilities|Casting|CastToFortPlayerController (Widget|GetOwningPlayer _returnvalue)))
        (Utilities|IsValid (Game|Player|GetPlayerPawn _asfort_player_controller)
          (:Is%20Valid
            (CallFunction|SetUIVisibility (Variables|Default|GetUI_InteractionPoint_Upgrade) ShowDetail ShowDistance)))))))

(event OnComponentEndOverlap(SphereDetail)
  (params (OverlappedComponent :type Primitive%20Component%20Object%20Reference) (OtherActor :type Actor%20Object%20Reference) (OtherComp :type Primitive%20Component%20Object%20Reference) (OtherBodyIndex :type Integer))
  (CallFunction|CheckShouldDisplayUI false true))

(event OnComponentBeginOverlap(CylinderDistance)
  (params (OverlappedComponent :type Primitive%20Component%20Object%20Reference) (OtherActor :type Actor%20Object%20Reference) (OtherComp :type Primitive%20Component%20Object%20Reference) (OtherBodyIndex :type Integer) (bFromSweep :type Boolean) (SweepResult :type Hit%20Result%20Structure%20%28by%20ref%29))
  (CallFunction|CheckShouldDisplayUI false true))

(event Custom|InitializeUI
  (bind _ashud_interaction_point (Utilities|Casting|CastToHUD_InteractionPoint (UserInterface|GetUserWidgetObject (Variables|Default|GetUI_InteractionPoint_Upgrade))))
  (Class|HUDInteractionPoint|EventSetText _ashud_interaction_point "NSLOCTEXT(\"[2D0636F544EC5881B9BEE08697266E3B]\", \"F696AC6340404E0EDF379C9AC28D7091\", \"\\r\\nUpgrade\")")
  (bind _ashud_interaction_point_2 (Utilities|Casting|CastToHUD_InteractionPoint (UserInterface|GetUserWidgetObject (Variables|Default|GetUI_InteractionPoint_Sidegrade))))
  (Class|HUDInteractionPoint|EventSetText _ashud_interaction_point_2 "NSLOCTEXT(\"[2D0636F544EC5881B9BEE08697266E3B]\", \"DD62981D4A7B8867E6A6CDA35047A276\", \"\\r\\nSidegrade\")"))

(event OnComponentEndOverlap(CylinderDistance)
  (params (OverlappedComponent :type Primitive%20Component%20Object%20Reference) (OtherActor :type Actor%20Object%20Reference) (OtherComp :type Primitive%20Component%20Object%20Reference) (OtherBodyIndex :type Integer))
  (CallFunction|CheckShouldDisplayUI false))

(event OnLinkedActorDestroyed(FortLinkToActor)
  (params (DamageCauser :type Actor%20Object%20Reference))
  ;[1 1 1] If the link to actor is broken try to link to something new. If nothing is found then destroy.
  (bind _returnvalue (LinkToActorComponent|PerformLinkingCheck (Variables|Default|GetFortLinkToActor) "Down" (Variables|Default|GetLinkToActor)))
  (if (not _returnvalue)
    (Actor|DestroyActor)))

(fn BlueprintGetFailedInteractionString
  (params (InteractingPawn :type Fort%20Pawn%20Object%20Reference) (InteractionBeingAttempted :type EInteractionBeingAttempted%20Enum :default "FirstInteraction"))
  (switch Utilities|FlowControl|Switch|SwitchonEFortWeaponUpgradeInteractionResult (Utilities|Select :Index InteractionBeingAttempted :FirstInteraction (Variables|BuildingItemWeaponUpgradeActor|GetUpgradeInteractionResult) :SecondInteraction (Variables|BuildingItemWeaponUpgradeActor|GetHorizontalUpgradeInteractionResult) :AllInteraction (Variables|BuildingItemWeaponUpgradeActor|GetUpgradeInteractionResult))
    (:NotEnoughResources
      (return (Variables|Strings|GetNotEnoughRessourcesText)))
    (:CannotUpgrade
      _)
    (:CannotInteract
      _))
  (return (Variables|Strings|GetWeaponCannotBeUpgradedText)))

(fn BlueprintGetInteractionString
  (params (InteractingPawn :type Fort%20Pawn%20Object%20Reference) (InteractionBeingAttempted :type EInteractionBeingAttempted%20Enum :default "FirstInteraction"))
  (return (Utilities|Select :Index InteractionBeingAttempted :FirstInteraction (Variables|Strings|GetFree%20String) :SecondInteraction (Variables|Strings|GetSideGradeString))))

(fn BlueprintGetInteractionTime
  (params (InteractingPawn :type Fort%20Pawn%20Object%20Reference) (InteractionBeingAttempted :type EInteractionBeingAttempted%20Enum :default "FirstInteraction"))
  (return (Utilities|Select :Index InteractionBeingAttempted :FirstInteraction (Utilities|Select :Index (Rarity|GetWeaponDataRarity (Variables|/Script/FortniteGame.FortPawn|GetCurrentWeapon InteractingPawn)) :Common (ScalableFloat|GetValueatLevel (Variables|Interaction|GetInteractionTimeUncommon)) :Uncommon (ScalableFloat|GetValueatLevel (Variables|Interaction|GetInteractionTimeRare)) :Rare (ScalableFloat|GetValueatLevel (Variables|Interaction|GetInteractionTimeVeryRare)) :Epic (ScalableFloat|GetValueatLevel (Variables|Interaction|GetInteractionTimeSuperRare)) :Legendary (ScalableFloat|GetValueatLevel (Variables|Interaction|GetInteractionTimeMythic)) :Mythic 0.0 :Transcendent 0.0 :Unattainable 0.0) :SecondInteraction (Utilities|Select :Index (Rarity|GetWeaponDataRarity (Variables|/Script/FortniteGame.FortPawn|GetCurrentWeapon InteractingPawn)) :Common (ScalableFloat|GetValueatLevel (Variables|Interaction|GetInteractionTimeSide)) :Uncommon (ScalableFloat|GetValueatLevel (Variables|Interaction|GetInteractionTimeSide) 1.0) :Rare (ScalableFloat|GetValueatLevel (Variables|Interaction|GetInteractionTimeSide) 2.0) :Epic (ScalableFloat|GetValueatLevel (Variables|Interaction|GetInteractionTimeSide) 3.0) :Legendary (ScalableFloat|GetValueatLevel (Variables|Interaction|GetInteractionTimeSide) 4.0) :Mythic 0.0 :Transcendent 0.0 :Unattainable 0.0) :AllInteraction 0.0) true))

(fn GetOverrideRarity
  (bind (_rarity _returnvalue) (|CallParentFunction))
  (Utilities|FlowControl|Sequence
    (:then_0
      (LocalVariables|SetNewRarity _rarity)
      (bind _output_get (LocalVariables|SetbHasOverride _returnvalue))
      (if _output_get
        _
        (elif (Variables|InteractMethods|InstanceOverride|GetbUseInstanceLootValueOverrides)
          (LocalVariables|SetNewRarity (Variables|InteractMethods|InstanceOverride|GetOverrideVendingMachineRarity))
          (LocalVariables|SetbHasOverride true))))
    (:then_1
      (return (LocalVariables|GetNewRarity) (LocalVariables|GetbHasOverride)))))

(fn LocalOnFailedInteract
  (params (InteractingPawn :type Fort%20Player%20Pawn%20Object%20Reference))
  (Audio|PlaySoundatLocation (Variables|VendingFX|GetVendFailedSound) (Transformation|GetActorLocation))
  (FailedAction|UnabletoPerformAction InteractingPawn true (Variables|Default|GetFailedReason)))

(fn SetLightColor
  (params (NewColor :type Linear%20Color%20Structure))
  (return))

(fn SetRarityColors
  (params (Color :type Linear%20Color%20Structure))
  (Rendering|Material|SetVectorParameterValue (Variables|VendingFX|GetMonitorMat) "ScreenColor" Color :!class /Script/Engine.MaterialInstanceDynamic)
  (Rendering|Material|SetVectorParameterValue (Variables|VendingFX|GetMetalMat) "PaintColor" Color :!class /Script/Engine.MaterialInstanceDynamic))

(fn SetUIVisibility
  (params (UITarget :type Widget%20Component%20Object%20Reference) (ShowDetail :type Boolean) (ShowDistance :type Boolean))
  (bind _ashud_interaction_point (Utilities|Casting|CastToHUD_InteractionPoint (UserInterface|GetUserWidgetObject UITarget)))
  (Class|HUDInteractionPoint|EventShowDetail _ashud_interaction_point ShowDetail)
  (if ShowDistance
    (if (Variables|/Script/Engine.SceneComponent|GetbHiddenInGame UITarget)
      (Development|SetHiddeninGame UITarget false false)
      (Class|HUDInteractionPoint|EventIntro _ashud_interaction_point))
    (else
      (if (not (Variables|/Script/Engine.SceneComponent|GetbHiddenInGame UITarget))
        (Development|SetHiddeninGame UITarget true false)))))

(fn UpdateAnimInstanceVisuals
  (bind _aswumba_gadget_anim_bp (Utilities|Casting|CastToWumba_Gadget_AnimBP (Components|SkeletalMesh|GetAnimInstance (Variables|Default|GetSK_Wumba))))
  (Class|WumbaGadgetAnimBP|UpdateUpgradingVisuals _aswumba_gadget_anim_bp (Variables|Default|GetUpgradeAnim)))

(fn ConstructionScript)
