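; Wumba deco tool data asset blueprint for placing the Wumba gameplay item in Athena.
; It overrides IsOnGround, which casts the instigator to a Fort player pawn and returns true when its character movement is not falling, or true if the cast fails.
; The press-trigger event and construction script are present but empty.
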(event DecoTool|EventOnPressTrigger
  (params (FortDecoHelper :type Fort%20Deco%20Helper%20Object%20Reference)))

(fn IsOnGround
  (bind _asfort_player_pawn (Utilities|Casting|CastToFortPlayerPawn (Game|GetInstigator))
    (:then
      (return (not (AI|Components|NavMovement|IsFalling (Variables|/Script/Engine.Character|GetCharacterMovement _asfort_player_pawn)))))
    (:CastFailed
      (return true))))

(fn ConstructionScript)
