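; Thrown Wumba item projectile that deploys the Wumba actor upright on near-flat ground.
; On server bounces, a sphere overlap check turns it into a pickup beside vehicles or hiding props to avoid physics issues; water entry also yields a pickup.
; If it stops on a vehicle, it resimulates with a set velocity.
; Non-server bounces play FX and sound, debounced by a retriggerable delay to prevent spam when sliding.
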(event EventOnBounce
  (params (Hit :type Hit%20Result%20Structure%20%28by%20ref%29))
  (bind (_returnvalue_roll _returnvalue_pitch _returnvalue_yaw) (Math|Rotator|BreakRotator (Transformation|GetActorRotation)))
  (bind (_bblockinghit _binitialoverlap _time _distance _location _impactpoint _normal _impactnormal _physmat _hitactor) (Collision|BreakHitResult Hit))
  (if (Networking|IsServer)
    ;[0.1451 1 0.3307 1 32] Check if near a vehicle or hiding prop, if so, spawn pickup and destroy self. This prevents unwanted physics on vehicle if there are a number of these actors around a vehicle or hiding prop: FORT-297225 - FORT-298789
    (bind _bvehiclecollisiondetected (CallFunction|CheckforCollisioninRadius _impactpoint))
    (if _bvehiclecollisiondetected
      (Inventory|SpawnPickupInWorld (Variables|WumbaProjectileDefaults|GetAGID) 1 _impactpoint _impactnormal 0 true true false -1 "Destruction" "ItemSpawner" 0 false)
      (CallFunction|SpawnedItemFx)
      (Delivery|Kill)
      (else
        (if (Math|Float|NearlyEqual(Float) (Math|Vector|DotProduct _impactnormal "0, 0,1.000000") 1.0 0.25)
          (Utilities|Casting|CastToFortVehicleInterface _hitactor
            (:CastFailed
              ;[1 1 1] Spawn Wumba
              (bind _returnvalue (GameplayTags|HasMatchingGameplayTag(Message) _hitactor (Variables|WumbaProjectileDefaults|GetWumbaTag)))
              (if (not _returnvalue)
                (Game|SpawnActorfromClass (Variables|WumbaProjectileDefaults|GetWumbaActor) (Math|Transform|MakeTransform (Transformation|GetActorLocation) (Math|Rotator|MakeRotator 0.0 0.0 (+ _returnvalue_yaw 9e+01)) "1.000000,1.000000,1.000000") "AlwaysSpawn" "SelectDefaultAtRuntime" (Game|GetInstigator) (Game|GetInstigator))
                (CallFunction|SpawnedItemFx)
                (Delivery|Kill)))))))
    (else
      (CallFunction|BounceFxEvent))))

(event Custom|BounceFxEvent
  ;[1 1 1] Fx
  (Utilities|FlowControl|Sequence
    (:then_0
      _)
    (:then_1
      (Utilities|FlowControl|RetriggerableDelay 0.3)))
  (Utilities|FlowControl|DoOnce false
    (:Completed
      ;;[1 1 1] Late fix so that the bounce fx don't spam when the projectile is sliding down a hill
      (Effects|SpawnEmitteratLocation (Variables|WumbaProjectileFX|GetBounceFX) (Transformation|GetActorLocation))
      (Audio|PlaySoundatLocation (Variables|WumbaProjectileFX|GetBounceSound) (Transformation|GetActorLocation)))))

(event Custom|SpawnedItemFx
  ;[1 1 1] Fx
  (Effects|SpawnEmitteratLocation (Variables|WumbaProjectileFX|GetSpawnFX) (Transformation|GetActorLocation) "90.000000, 0, 0" "2.000000,2.000000,2.000000"))

(event EventOnStop
  (params (Hit :type Hit%20Result%20Structure%20%28by%20ref%29))
  (bind (_bblockinghit _binitialoverlap _time _distance _location _impactpoint _normal _impactnormal _physmat _hitactor) (Collision|BreakHitResult Hit))
  (bind _projectilemovementcomponent (Variables|Movement|GetProjectileMovementComponent))
  ;[1 1 1] If projectile settles on a vehicle, resimulate again and impulse
  (Development|PrintString "STOP" true true "(R=1.000000,G=0.012011,B=0.000000,A=1.000000)" 6.0)
  (bind _returnvalue (GameplayTags|HasMatchingGameplayTag(Message) _hitactor (Variables|WumbaProjectileDefaults|GetWumbaTag)))
  (if (not _returnvalue)
    (Utilities|Casting|CastToFortVehicleInterface _hitactor)
    (Components|Movement|SetUpdatedComponent _projectilemovementcomponent (Variables|Transformation|GetRootComponent))
    (Variables|/Script/Engine.ProjectileMovementComponent|SetbSimulationEnabled true _projectilemovementcomponent)
    (Components|Movement|SetUpdatedComponent _projectilemovementcomponent (Variables|Transformation|GetRootComponent))
    (Game|Components|ProjectileMovement|SetVelocityinLocalSpace _projectilemovementcomponent (Variables|WumbaProjectileDefaults|GetResimulateVelocity))))

(event OnEnterWaterEvent(WaterInteractionComponent)
  (params (WaterBody :type Fort%20Water%20Body%20Actor%20Object%20Reference) (WaterInteractionComponent :type Fort%20Water%20Interaction%20Component%20Object%20Reference) (bIsFirstBody :type Boolean))
  ;[1 1 1] Spawn PickUp if entering water
  (if (Networking|IsServer)
    (Inventory|SpawnPickupInWorld (Variables|WumbaProjectileDefaults|GetAGID) 1 (Transformation|GetActorLocation) "0.000000,0.000000,1.000000" 0 false true false -1 "Tossed" "Unset" 0 false)
    (CallFunction|WaterImpactItemFX)
    (Delivery|Kill)))

(event Custom|WaterImpactItemFX
  ;[1 1 1] Fx
  (Effects|SpawnEmitteratLocation :EmitterTemplate (Variables|WumbaProjectileFX|GetSpawnFX_Water) :Location (Transformation|GetActorLocation) :Scale "0.800000,0.800000,0.800000" :bAutoDestroy true :PoolingMethod "None" :bAutoActivateSystem true)
  (Audio|PlaySoundatLocation (Variables|WumbaProjectileFX|GetSplash%20Sound%20) (Transformation|GetActorLocation)))

(fn CheckForCollisionInRadius
  (params (SpherePos :type Vector))
  (bind (_outactors _returnvalue) (Collision|SphereOverlapActors SpherePos (Variables|WumbaProjectileDefaults|GetCollisionRadiusCheck) (Utilities|Array|MakeArray "ObjectTypeQuery5" "ObjectTypeQuery1")))
  (bind _output_get (Variables|WumbaProjectileDynamics|SetCollisionWith _outactors))
  (if _returnvalue
    (if (> (Utilities|Array|Length _output_get) 0)
      (for _array_element (Variables|WumbaProjectileDynamics|GetCollisionWith)
        (Utilities|Casting|CastToB_HidingProp _array_element
          (:then
            _)
          (:CastFailed
            (Utilities|Casting|CastToFortVehicleInterface _array_element
              (:then
                _)
              (:CastFailed
                (return false)))))
        (return true))
      (else
        _))
    (else
      _))
  (return false))

(fn ConstructionScript
  (|CallParentFunction))
