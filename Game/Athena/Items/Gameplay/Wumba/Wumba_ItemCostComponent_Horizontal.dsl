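; Item cost component for the Wumba weapon upgrade actor that reports the resource price of its horizontal upgrade option.
; GetCost casts the owner to B_Athena_Wumba and reads its required horizontal resources.
; When those are valid, it returns a map from Wood, Stone and Metal to the required wood, brick and metal amounts, otherwise a fallback value.
; BeginPlay and Tick are disabled stubs.
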(event EventBeginPlay
  (|CallParentFunction) ; This node is disabled and will not be called.\nDrag off pins to build functionality.
)

(event EventTick
  (params (DeltaSeconds :type Float%20%28single-precision%29))
  (|CallParentFunction DeltaSeconds) ; This node is disabled and will not be called.\nDrag off pins to build functionality.
)

(fn GetCost
  (bind _asb_athena_wumba (Utilities|Casting|CastToB_Athena_Wumba (Components|GetOwner)))
  (if (Utilities|IsValid (Variables|/Script/FortniteGame.BuildingItemWeaponUpgradeActor|GetRequiredHorizontalResources _asb_athena_wumba) :!class /Script/Engine.KismetSystemLibrary)
    (return (Utilities|Map|MakeMap (Utilities|Enum|LiteralenumEFortResourceType "Wood") (Variables|/Script/FortniteGame.WeaponUpgradeRequiredResources|GetRequiredWood (Variables|/Script/FortniteGame.BuildingItemWeaponUpgradeActor|GetRequiredHorizontalResources _asb_athena_wumba)) (Utilities|Enum|LiteralenumEFortResourceType "Stone") (Variables|/Script/FortniteGame.WeaponUpgradeRequiredResources|GetRequiredBrick (Variables|/Script/FortniteGame.BuildingItemWeaponUpgradeActor|GetRequiredHorizontalResources _asb_athena_wumba)) (Utilities|Enum|LiteralenumEFortResourceType "Metal") (Variables|/Script/FortniteGame.WeaponUpgradeRequiredResources|GetRequiredMetal (Variables|/Script/FortniteGame.BuildingItemWeaponUpgradeActor|GetRequiredHorizontalResources _asb_athena_wumba))))
    (else
      (return "Wood"))))
